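# Physics asset for the proxy microwave animation prop in the Donut_2 set, giving it a simple two-body physical shape.
# A larger capsule on the Root bone approximates the main body, and a smaller offset capsule on the Door bone covers the door.
# A single constraint joins Door to Root with equal swing and twist limit angles, rather than values specific to a one-axis hinge.
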PhysicsAsset /Game/Animation/Props/Donut_2/Proxy/Proxy_Microwave_Prop_PhysicsAsset
bodies {
 Door [{shapeType=Capsule, length=23.3855, radius=15.1599, center=<-23.1304 0.4678 -0.4904>, rotation=<rot 0 93.1674 -89.9757>}]
 Root [{shapeType=Capsule, length=44.1865, radius=35.2087, center=<9.0398 5.2446 -2.1874>, rotation=<rot 0 -54.1803 -75.1301>}]
}
constraints [{bone1Name=Door, bone2Name=Root, swing1LimitDegrees=45, swing2LimitDegrees=45, twistLimitDegrees=45}]
